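# Translucent, two-sided, unlit ice material for Athena winter art, faking lighting and internal depth.
# Color sums a tinted, desaturated sky-cubemap reflection, normal-mapped Fresnel and blue tints blended by a scratch mask.
# Two BumpOffset depths parallax the scratch normal and mask, while vertex red flattens normals and shifts color.
# Opacity nodes derive thickness from scene-minus-pixel depth, with camera-distance-scaled falloff; root connections are not shown.
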Material /Figment_S09/Athena/WinterArt/Materials/M_Athena_IceTransparent_Cattus
graph {
 properties {
  bUsedWithInstancedStaticMeshes=true
  blendMode=BLEND_Translucent
  pixelDepthOffsetMode=PDOM_Legacy
  refractionMethod=RM_PixelNormalOffset
  shadingModel=MSM_Unlit
  translucencyLightingMode=TLM_VolumetricDirectional
  twoSided=true
  usedShadingModels=Unlit
 }
 nodes [
  n1 = Desaturation(
   None=TextureSample(UVs=ReflectionVectorWS()) {texture=/Game/Packages/Fortress_Sky/SkyDome/MaterialInstances/SkyDome_S8/Cubemap_S8_Day}.RGB
   Fraction=ScalarParameter() {defaultValue=1, parameterName=ReflectionsDesat}
  )
  n2 = VectorParameter() {parameterName=Color4, defaultValue=<rgba 0.1795 0.4268 0.5729 1>}
  n3 = Multiply(A=TextureCoordinate(), B=ScalarParameter() {defaultValue=0.6877, parameterName=Tiling})
  n4 = Constant() {r=1}
  n5 = TextureSample(UVs=BumpOffset(Coordinate=@n3, Height=@n4, HeightRatioInput=ScalarParameter() {defaultValue=-0.0405, parameterName=DepthOffset})) {samplerType=SAMPLERTYPE_Normal, texture=/Game/Athena/WinterArt/Textures/T_IceScratch1_N}
  n6 = Constant3Vector() {constant=<rgba 0 0 1 0>}
  n7 = VertexColor()
  n8 = Fresnel(
   ExponentIn=ScalarParameter() {defaultValue=1.846, parameterName=Fresnel}
   Normal=Transform(None=LinearInterpolate(A=@n5.RGB, B=@n6, Alpha=@n7.R))
  )
  n9 = ScalarParameter() {defaultValue=0.0571, parameterName=FresnelBirghtness}
  n10 = VectorParameter() {parameterName=Color, defaultValue=<rgba 0.0596 0.1445 0.1875 1>}
  n11 = VectorParameter() {parameterName=Color2, defaultValue=<rgba 0.0278 0.1017 0.2812 1>}
  n12 = TextureSample(UVs=BumpOffset(Coordinate=@n3, Height=@n4, HeightRatioInput=ScalarParameter() {defaultValue=-0.5167, parameterName=DepthOffset2})) {samplerType=SAMPLERTYPE_Masks, texture=/Figment_S09/Athena/WinterArt/Textures/T_IceScratch1_M}
  Add(
   A=Add(A=Multiply(A=@n1, B=@n2.RGB), B=Multiply(A=@n8, B=@n9))
   B=LinearInterpolate(
    A=LinearInterpolate(A=@n10.RGB, B=@n11.RGB, Alpha=@n12.R)
    B=VectorParameter() {parameterName=Color3, defaultValue=<rgba 0.0066 0.0356 0.0469 1>}.RGB
    Alpha=@n7.R
   )
  )
  Constant3Vector()
  Constant()
  n13 = ScalarParameter() {defaultValue=0.1048, parameterName=Opacity}
  n14 = Subtract(A=SceneDepth().R, B=PixelDepth().R)
  n15 = ScalarParameter() {defaultValue=870.4197, parameterName=asdasdas}
  n16 = ScalarParameter() {defaultValue=28000, parameterName=Distance}
  n17 = LinearInterpolate(
   A=@n15
   B=ScalarParameter() {defaultValue=50, parameterName=asdasdas2}
   Alpha=MaterialFunctionCall(Fade Length=@n16, Fade Offset=ScalarParameter() {defaultValue=1000, parameterName=DistanceOffset}) {materialFunction=&CameraDepthFade}.Result
  )
  n18 = LinearInterpolate(A=@n13, B=Constant() {r=1}, Alpha=Saturate(None=Divide(A=@n14, B=@n17)))
  DepthFade(Opacity=@n18, FadeDistance=ScalarParameter() {defaultValue=263.9487, parameterName=DepthFade})
  n19 = DotProduct(A=CameraVectorWS(), B=VertexNormalWS())
  Divide(A=@n16, B=@n19)
  n20 = ScalarParameter() {defaultValue=0.2667, parameterName=Opacity2}
  LinearInterpolate(A=@n18, B=@n20, Alpha=@n7.R)
  LinearInterpolate(A=@n13, B=@n20, Alpha=Fresnel(Normal=VertexNormalWS()) {baseReflectFraction=0, exponent=3}) {constB=2}
  MaterialFunctionCall(
   Normal=@n5.RGB
   Flatness=LinearInterpolate(A=ScalarParameter() {defaultValue=0.75, parameterName=NormalFlatten}, B=Constant() {r=1}, Alpha=@n7.R)
  ) {materialFunction=&FlattenNormal}
  Multiply(B=ScalarParameter() {defaultValue=0.6476, parameterName=Reflections})
  ScalarParameter() {parameterName=Roughness}
  ScalarParameter() {defaultValue=0.8571, parameterName=Refraction}
  Subtract(A=@n14, B=Divide(A=@n15, B=@n19))
  TextureSample() {samplerType=SAMPLERTYPE_Masks, texture=&T_IceScratch1_M}
  VectorParameter() {defaultValue=<rgba 0.0126 0.08 0.1615 0>}
  VertexNormalWS()
 ]
}
